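; Anim notify state for the DonkeyCrib emote that drives portal material effects on the animated skeletal mesh.
; On begin it enables render custom depth on the mesh, and no end handler resets it.
; Each tick it reads the world transform of the JNT_Portal socket and pushes position, up vector and X scale as opacity into material parameters.
; It also sends the IsUp and IsPortalAnimating variables as scalar facing and visibility flags.
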(fn Received_NotifyBegin
  (params (MeshComp :type Skeletal%20Mesh%20Component%20Object%20Reference) (Animation :type Anim%20Sequence%20Base%20Object%20Reference) (TotalDuration :type Float%20%28single-precision%29) (EventReference :type Anim%20Notify%20Event%20Reference%20Structure%20%28by%20ref%29))
  (Rendering|SetRenderCustomDepth MeshComp true)
  (return false))

(fn Received_NotifyTick
  (params (MeshComp :type Skeletal%20Mesh%20Component%20Object%20Reference) (Animation :type Anim%20Sequence%20Base%20Object%20Reference) (FrameDeltaTime :type Float%20%28single-precision%29) (EventReference :type Anim%20Notify%20Event%20Reference%20Structure%20%28by%20ref%29))
  (bind (_returnvalue_location _returnvalue_rotation _returnvalue_scale) (Math|Transform|BreakTransform (Transformation|GetSocketTransform MeshComp "JNT_Portal" "RTS_World")))
  (bind _returnvalue_scale_x (Math|Vector|BreakVector _returnvalue_scale))
  (Utilities|FlowControl|Sequence
    (:then_0
      (Rendering|Material|SetVectorParameterValueonMaterials MeshComp "PortalCenterPosition" _returnvalue_location))
    (:then_1
      (Rendering|Material|SetScalarParameterValueonMaterials MeshComp "IsPortalFacingUp" (Math|Conversions|ToFloat(Boolean) (Variables|Default|GetIsUp))))
    (:then_2
      (Rendering|Material|SetScalarParameterValueonMaterials MeshComp "IsPortalShowing" (Math|Conversions|ToFloat(Boolean) (Variables|Default|GetIsPortalAnimating))))
    (:then_3
      (Rendering|Material|SetScalarParameterValueonMaterials MeshComp "PortalOpacity" _returnvalue_scale_x))
    (:then_4
      (Rendering|Material|SetVectorParameterValueonMaterials MeshComp "PortalBoneOrientation" (Math|Vector|GetUpVector _returnvalue_rotation))
      (return false))))
